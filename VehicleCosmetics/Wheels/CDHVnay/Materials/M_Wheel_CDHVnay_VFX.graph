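# Default Lit VFX material for the CDHVnay cosmetic wheel, enabled for skeletal meshes and mesh deformers, that builds a view-dependent red and pink sheen.
# Two inverted Fresnel terms, using MF_Fresnel_Classic from the Cola_Holographic wheel with constant Z normals, tint red and pink layers with separate exponent and balance parameters.
# Both layers are added onto a dark red base, alongside a Roughness parameter, with no textures.
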Material /VehicleCosmetics/Wheels/CDHVnay/Materials/M_Wheel_CDHVnay_VFX
graph {
 properties {
  bAutomaticallySetUsageInEditor=false
  bUsedWithMeshDeformer=true
  bUsedWithSkeletalMesh=true
  pixelDepthOffsetMode=PDOM_Legacy
  previewMesh=/VehicleCosmetics/Wheels/CDHVnay/Meshes/SK_Wheel_CDHVnay
  usedShadingModels=Default Lit
 }
 nodes [
  n1 = Constant3Vector() {constant=<rgba 0 0 1 0>}
  n2 = ScalarParameter() {defaultValue=1, parameterName=RedFresnel}
  n3 = Multiply(
   A=Constant3Vector() {constant=<rgba 1 0 0 0>}
   B=OneMinus(None=MaterialFunctionCall!(Normal=@n1, Exponent=@n2) {materialFunction=/VehicleCosmetics/Wheels/Cola_Holographic/Materials/MF_Fresnel_Classic}.Result)
  )
  n4 = ScalarParameter() {defaultValue=2, parameterName=RedBalance}
  n5 = Constant3Vector() {constant=<rgba 1 0 0.1 0>}
  n6 = OneMinus(None=MaterialFunctionCall(Normal=Constant3Vector() {constant=<rgba 0 0 1 0>}, Exponent=ScalarParameter() {defaultValue=0.75, parameterName=PinkFresnel}) {materialFunction=&MF_Fresnel_Classic}.Result)
  Add(
   A=Add(A=Constant3Vector() {constant=<rgba 0.06 0 0 0>}, B=Multiply(A=@n3, B=@n4))
   B=Multiply(A=Multiply(A=@n5, B=@n6), B=ScalarParameter() {defaultValue=2, parameterName=PinkBalance})
  )
  ScalarParameter() {parameterName=Roughness}
 ]
}
